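; Base Blueprint for an interactable hot air balloon transit station sign that lets players highlight the next incoming balloon.
; It always allows interaction and shows a "Highlight Next Balloon" prompt.
; On interact, it checks its soft train rail reference, then calls MarkUpcomingTrainforStation on that FortTrainRail with the player's controller and the sign's station identifier.
; Balloons reuse the train rail system.
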(event Interaction|EventBlueprintOnInteract
  (params (InteractingPawn :type Fort%20Pawn%20Object%20Reference) (InteractionBeingAttempted :type EInteractionBeingAttempted%20Enum :default "FirstInteraction"))
  (bind _trainrail (Variables|Default|GetTrainRail))
  (bind _asfort_player_pawn (Utilities|Casting|CastToFortPlayerPawn InteractingPawn))
  (if (Utilities|IsValidSoftObjectReference _trainrail)
    (Class|FortTrainRail|MarkUpcomingTrainforStation (Utilities|ResolveSoftReference _trainrail) (Fortnite|GetFortPlayerControllerfromActor _asfort_player_pawn) (Variables|Default|GetTrainStationIdentifier))))

(fn BlueprintCanInteract
  (params (InteractingPawn :type Fort%20Pawn%20Object%20Reference) (InteractionBeingAttempted :type EInteractionBeingAttempted%20Enum :default "FirstInteraction") (InteractionType :type TInteractionType%20Enum :default "IT_NoInteraction"))
  (return true))

(fn BlueprintGetInteractionString
  (params (InteractingPawn :type Fort%20Pawn%20Object%20Reference) (InteractionBeingAttempted :type EInteractionBeingAttempted%20Enum :default "FirstInteraction"))
  (return "NSLOCTEXT(\"[50EA01C1EFCD2E7667A831776BDC60F3]\", \"ADAF23EC4309C190EBA4D391FDC5A229\", \"Highlight Next Balloon\")"))

(fn ConstructionScript
  (|CallParentFunction))
